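# Masked, two-sided subsurface material for Niagara mesh particles that draws water chevron shapes in the WaveMash wave effect.
# Distorted, world-projected noise offsets a curve-atlas color lookup and, with vertex red, forms an erodible mask, dithered using TAA.
# Dynamic parameters drive panning, erosion and camera-relative Z offset; noise also displaces vertices, and the mask drives pixel depth offset.
# Mobile pans color noise by time and drops masking and depth offset; the Darken multiply is unconnected.
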Material /Game/Effects/Niagara/Emitters/Water/WaveMash/M_NFX_ChevronShapes
graph {
 properties {
  bTangentSpaceNormal=false
  bUseMaterialAttributes=true
  bUsedWithNiagaraMeshParticles=true
  blendMode=BLEND_Masked
  opacityMaskClipValue=0.01
  pixelDepthOffsetMode=PDOM_Legacy
  shadingModel=MSM_Subsurface
  twoSided=true
  usedShadingModels=Subsurface
 }
 nodes [
  n1 = TextureCoordinate()
  n2 = Divide(A=WorldPosition().XYZ, B=ScalarParameter() {defaultValue=256, parameterName=WorldTextureScale}) {constB=2048}
  n3 = Add(A=@n2, B=Multiply(A=Time()) {constB=0.25})
  n4 = TextureCoordinate() {uTiling=0.5, vTiling=0.5}
  n5 = TextureSample(UVs=Add(A=ComponentMask(None=@n3) {g=true, r=true}, B=TextureSample(UVs=@n4) {texture=&T_NoisePack_01}.R)) {texture=&T_NoisePack_01}
  n6 = CurveAtlasRowParameter(CurveTime=Add(A=ComponentMask(None=@n1) {g=true}, B=Multiply(A=@n5.G) {constB=0.25})) {
   atlas=/Game/Effects/Niagara/Emitters/Water/WaterAtlas
   curve=&/Curve_ChevronShapeColor
   defaultValue=18
  }
  Multiply(A=@n6, B=DynamicParameter() {defaultValue=<rgba 1 0 1 1>, paramNames [TexturePannerTime, Erode_Subtract, Z_Offset, Darken]}.Darken)
  n7 = Constant3Vector() {constant=<rgba 0 0 1 0>}
  n8 = TextureCoordinate()
  n9 = Add(
   A=@n2
   B=DynamicParameter() {defaultValue=<rgba 1 0 1 1>, paramNames [TexturePannerTime, Erode_Subtract, Z_Offset, Darken]}.TexturePannerTime
  )
  n10 = TextureCoordinate() {uTiling=0.5, vTiling=0.5}
  n11 = TextureSample!(
   UVs=Add(
    A=ComponentMask(None=@n9) {g=true, r=true}
    B=TextureSample(UVs=@n10) {mipValueMode=TMVM_MipLevel, texture=&T_NoisePack_01}.R
   )
   Tex
   Level
   Apply View MipBias
  ) {mipValueMode=TMVM_MipLevel, texture=&T_NoisePack_01}
  n12 = CurveAtlasRowParameter(CurveTime=Add(A=ComponentMask(None=@n8) {g=true}, B=Multiply(A=@n11.G) {constB=0.25})) {
   atlas=&WaterAtlas
   curve=&Curve_ChevronShapeColor
   defaultValue=18
  }
  n13 = Constant() {r=0.25}
  n14 = Constant()
  n15 = VertexColor()
  n16 = Subtract(
   A=Multiply(A=SquareRoot(None=@n15.R), B=Add(A=@n11.R))
   B=DynamicParameter() {defaultValue=<rgba 1 0 1 1>, paramNames [TexturePannerTime, Erode_Subtract, Z_Offset, Darken]}.Erode_Subtract
  )
  n17 = Custom!(None) {
   code=View.TemporalAAParams
   desc="\n*** Original source before expression upgrade ***\nView.TemporalAAParams\n*** Original source before expression upgrade ***\nFrame.TemporalAAParams"
   description=TemporalAA SampleID
   outputType=CMOT_Float2
  }
  MaterialExpressionAdd_2 = Add(
   A=MaterialFunctionCall!(TextureResolution=Constant2Vector() {g=1, r=1}) {materialFunction=/Engine/Functions/Engine_MaterialFunctions02/Texturing/ScreenAlignedPixelToPixelUVs}.Result
   B=ComponentMask(None=@n17) {r=true}
  )
  n18 = Custom!(p=@MaterialExpressionAdd_2) {
   code="((uint)(p.x) + 2 * (uint)(p.y)) % 5"
   outputType=CMOT_Float1
   inputs [{inputName=p, input {expression=&self:MaterialExpressionAdd_2}}]
  }
  n19 = TextureSample(UVs=MaterialFunctionCall(TextureResolution=Constant2Vector() {g=64, r=64}) {materialFunction=&ScreenAlignedPixelToPixelUVs}.Result) {mipValueMode=TMVM_MipLevel, samplerType=SAMPLERTYPE_LinearColor, texture=/Game/Effects/Fort_Effects/Textures/Tile/LinerNoise64x64}
  n20 = Add(A=@n18, B=@n19.R) {constB=0}
  n21 = Add(A=Add(A=@n16, B=Divide(A=@n20) {constB=6}) {constB=0.5}) {constB=-0.5}
  n22 = Saturate(None=@n21)
  n23 = ParticleColor()
  n24 = ParticlePositionWS()
  n25 = ObjectPositionWS()
  n26 = ComponentMask(None=Subtract(A=CameraPositionWS(), B=ShadingPathSwitch(Default=@n24, Mobile=@n25) {desc=Migrated from FeatureLevelSwitch})) {b=true}
  n27 = Constant2Vector()
  n28 = DynamicParameter() {defaultValue=<rgba 1 0 1 1>, paramNames [TexturePannerTime, Erode_Subtract, Z_Offset, Darken]}
  n29 = ComponentMask(None=@n2) {g=true, r=true}
  MaterialExpressionAdd_8 = Add(
   A=Multiply(A=Clamp(None=@n26) {minDefault=-1}, B=AppendVector(A=@n27, B=@n28.Z_Offset))
   B=Multiply(A=VertexColor().R, B=TextureSample(UVs=@n29) {mipValueMode=TMVM_MipLevel, texture=&T_NoisePack_01}.B)
  )
  n30 = Custom!(Value=@MaterialExpressionAdd_8) {
   code="##if XBOXONE_PROFILE || SM5_PROFILE || SM4_PROFILE\r\n\tprecise float3 pv = Value;\r\n\treturn pv;\r\n##else\r\n\treturn Value;\r\n##endif"
   desc=Make Precise to Fix Depth Variance
   inputs [{inputName=Value, input {expression=&self:MaterialExpressionAdd_8}}]
  }
  n31 = Multiply(A=@n21, B=ScalarParameter() {defaultValue=64, parameterName=PixelDepthOffset_RandomRange}) {constB=512}
  n32 = Add(A=ScalarParameter() {defaultValue=20, parameterName=PixelDepthOffset_CameraOffset}) {constB=180}
  n33 = Multiply(A=@n16, B=ParticleColor().A)
  ShadingPathSwitch(
   Default=MakeMaterialAttributes!(
    BaseColor=@n12
    Metallic
    Specular=@n13
    Roughness=@n14
    Anisotropy
    EmissiveColor
    Opacity=@n33
    OpacityMask=Multiply(A=@n22, B=@n23.A)
    Normal=@n7
    Tangent
    WorldPositionOffset=@n30
    SubsurfaceColor=VectorParameter() {parameterName=SSS, defaultValue=<rgba 0.2178 0.297 0.3802 1>}.RGB
    ClearCoat
    ClearCoatRoughness
    AmbientOcclusion
    Refraction
    CustomizedUVs_0
    CustomizedUVs_1
    CustomizedUVs_2
    CustomizedUVs_3
    CustomizedUVs_4
    CustomizedUVs_5
    CustomizedUVs_6
    CustomizedUVs_7
    PixelDepthOffset=Subtract(A=@n31, B=@n32)
    ShadingModel
    Displacement
   )
   Mobile=MakeMaterialAttributes(BaseColor=@n6, Specular=@n13, Roughness=@n14, Opacity=@n33, OpacityMask=Constant() {r=1}, Normal=@n7, WorldPositionOffset=@n30)
  ) {desc=Migrated from FeatureLevelSwitch}
 ]
}
